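# Physics asset for the semi truck's ladder damage part, simulating a ladder and left exhaust as jointed capsule bodies.
# Four bodies form a chain from ladder through body and frame to the left exhaust, with identical small capsules on body and frame.
# The body-to-frame joint swings and twists freely, while the ladder-to-body and exhaust-to-frame joints have angular limits of only a few degrees.
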PhysicsAsset /Valet/Semi_Truck/Meshes/Semi_Truck_Veh_Dyn_dmg_ladder
bodies {
 body [{shapeType=Capsule, length=11.9688, radius=10.4375, center=<-3.643 0.0005 -0.0002>, rotation=<rot 0 0 89.9997>}]
 exhaust_l [{shapeType=Capsule, length=177.25, radius=18.5938, center=<0 -17.8475 84.1043>}]
 frame [{shapeType=Capsule, length=11.9688, radius=10.4375, center=<-3.643 0.0005 -0.0002>, rotation=<rot 0 0 89.9997>}]
 ladder [{shapeType=Capsule, length=146.1562, radius=6.4062, center=<-54.379 0 -94.1344>}]
}
constraints [
 {bone1Name=ladder, bone2Name=body, swing1LimitDegrees=5, swing2LimitDegrees=5, twistLimitDegrees=5}
 {
  bone1Name=body
  bone2Name=frame
  swing1LimitDegrees=45
  swing1Motion=Free
  swing2LimitDegrees=45
  swing2Motion=Free
  twistLimitDegrees=45
  twistMotion=Free
 }
 {bone1Name=exhaust_l, bone2Name=frame, swing1LimitDegrees=4, swing2LimitDegrees=2, twistLimitDegrees=2}
]
