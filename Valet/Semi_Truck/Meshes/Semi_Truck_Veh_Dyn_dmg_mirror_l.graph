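# Physics asset for the semi truck's left-mirror damage mesh, judging by its name, with capsule bodies on the body, frame, right exhaust and two left mirror bones.
# Constraints chain the outer mirror bone mirror_01_l to mirror_l, then to body, with angular limits at both mirror joints.
# Body joins frame with free swing and twist, while the exhaust hangs off frame with only a few degrees of play.
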PhysicsAsset /Valet/Semi_Truck/Meshes/Semi_Truck_Veh_Dyn_dmg_mirror_l
bodies {
 body [{shapeType=Capsule, length=77.8438, radius=40, center=<-3.643 0.0005 -0.0002>, rotation=<rot 0 0 89.9997>}]
 exhaust_r [{shapeType=Capsule, length=208.0938, radius=18.5938, center=<0 16.8259 133.4946>}]
 frame [{shapeType=Capsule, length=77.8438, radius=40, center=<-3.643 0.0005 -0.0002>, rotation=<rot 0 0 89.9997>}]
 mirror_01_l [{shapeType=Capsule, length=0, radius=10, center=<-3.9534 -1.7033 -13.7911>, rotation=<rot 0 0 90>}]
 mirror_l [{shapeType=Capsule, length=33.75, radius=10, center=<7.3689 -21.2923 25.5857>, rotation=<rot 0 0 179.9995>}]
}
constraints [
 {bone1Name=mirror_l, bone2Name=body, swing1LimitDegrees=30, swing2LimitDegrees=10, twistLimitDegrees=60}
 {bone1Name=mirror_01_l, bone2Name=mirror_l, swing1LimitDegrees=45, swing2LimitDegrees=45, twistLimitDegrees=45}
 {
  bone1Name=body
  bone2Name=frame
  swing1LimitDegrees=45
  swing1Motion=Free
  swing2LimitDegrees=45
  swing2Motion=Free
  twistLimitDegrees=45
  twistMotion=Free
 }
 {bone1Name=exhaust_r, bone2Name=frame, swing1LimitDegrees=5, swing2LimitDegrees=3, twistLimitDegrees=3}
]
