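# Physics asset for the semi truck's mudflaps, apparently for a dynamic damage mesh, as the asset name suggests.
# A small capsule on the body bone anchors identical left and right mudflap capsules, each offset down from its bone to hang below the pivot.
# Each flap is constrained to the body with a wide primary swing and tight secondary swing and twist limits, so it flaps mostly along one axis like a hinge.
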PhysicsAsset /Valet/Semi_Truck/Meshes/Semi_Truck_Veh_Dyn_dmg_mudflaps
bodies {
 body [{shapeType=Capsule, length=11.9688, radius=10.4375, center=<-3.643 0.0005 -0.0002>, rotation=<rot 0 0 89.9997>}]
 mudflap_l [{shapeType=Capsule, length=39.1562, radius=15.5312, center=<0 0.0004 -40.5055>}]
 mudflap_r [{shapeType=Capsule, length=39.1562, radius=15.5312, center=<0 0.0004 -40.5055>}]
}
constraints [
 {bone1Name=mudflap_r, bone2Name=body, swing1LimitDegrees=60, swing2LimitDegrees=10.7143, twistLimitDegrees=5}
 {bone1Name=mudflap_l, bone2Name=body, swing1LimitDegrees=60, swing2LimitDegrees=10.7143, twistLimitDegrees=5}
]
