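# Base AnimBlueprint for the semi truck vehicle, procedurally posing its axles, wheels and steering wheel from the reference pose.
# A chain of bone transforms feeds RigidBody physics for mudflaps, plus a ladder and mirrors enabled at health thresholds.
# The update event sets those triggers and, for FortDagwoodVehicle, detects the offroad tire mod and offsets the middle differential from the rear axle.
# That mod flag enables additive bone translations that reposition all wheels.
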AnimBlueprint /Valet/Semi_Truck/Meshes/Semi_Truck_Veh_AnimBP_Base
target_skeleton=&/Semi_Truck_Veh_Skeleton
graphs [
 AnimGraph {nodes [
   n1 = Get AxlePositionF() -> (AxlePositionF:Vector)
   n2 = Get AxleRotationF() -> (AxleRotationF:Rotator)
   n3 = "Transform (Modify) Bone - Bone: axel_pivot_fr_l"(Alpha=1, Translation=@n1, Rotation=@n2, Scale="1,1,1") {
    rotationMode=BMM_Replace
    rotationSpace=BCS_BoneSpace
    translationMode=BMM_Replace
    translationSpace=BCS_BoneSpace
    alphaBoolBlend {blendOption=Linear}
    boneToModify {boneName=axel_pivot_fr_l}
   }
   "Transform (Modify) Bone - Bone: axel_pivot_fr_r"(ComponentPose=@n3, Alpha=1, Translation=@n1, Rotation=@n2, Scale="1,1,1") {
    rotationMode=BMM_Replace
    rotationSpace=BCS_BoneSpace
    translationMode=BMM_Replace
    translationSpace=BCS_BoneSpace
    alphaBoolBlend {blendOption=Linear}
    boneToModify {boneName=axel_pivot_fr_r}
   }
   n4 = "Transform (Modify) Bone - Bone: front_differential"(ComponentPose=Local To Component(LocalPose=Local Space Ref Pose()), Alpha=1, Translation=@n1, Rotation=@n2, Scale="1,1,1") {
    rotationMode=BMM_Replace
    rotationSpace=BCS_BoneSpace
    translationMode=BMM_Replace
    translationSpace=BCS_BoneSpace
    alphaBoolBlend {blendOption=Linear}
    boneToModify {boneName=front_differential}
   }
   n5 = Get AxleRotationB() -> (AxleRotationB:Rotator)
   n6 = "Transform (Modify) Bone - Bone: rear_differential"(ComponentPose=@n4, Alpha=1, Translation=Get AxlePositionB() -> (AxlePositionB:Vector), Rotation=@n5, Scale="1,1,1") {
    rotationMode=BMM_Replace
    rotationSpace=BCS_BoneSpace
    translationMode=BMM_Replace
    translationSpace=BCS_BoneSpace
    alphaBoolBlend {blendOption=Linear}
    boneToModify {boneName=rear_differential}
   }
   n7 = "Transform (Modify) Bone - Bone: mid_differential"(ComponentPose=@n6, Alpha=1, Translation=Get Mid Differential Position() -> (Mid Differential Position:Vector), Rotation=@n5, Scale="1,1,1") {
    rotationMode=BMM_Replace
    rotationSpace=BCS_BoneSpace
    translationMode=BMM_Replace
    translationSpace=BCS_BoneSpace
    alphaBoolBlend {blendOption=Linear}
    boneToModify {boneName=mid_differential}
   }
   n8 = "Transform (Modify) Bone - Bone: wheel_steering_fr_r"(ComponentPose=@n7, Alpha=1, Translation="0,0,0", Rotation=Get WheelRotationFR() -> (WheelRotationFR:Rotator), Scale="1,1,1") {
    rotationMode=BMM_Replace
    rotationSpace=BCS_BoneSpace
    translationSpace=BCS_BoneSpace
    alphaBoolBlend {blendOption=Linear}
    boneToModify {boneName=wheel_steering_fr_r}
   }
   n9 = "Transform (Modify) Bone - Bone: wheel_steering_fr_l"(ComponentPose=@n8, Alpha=1, Translation="0,0,0", Rotation=Get WheelRotationFL() -> (WheelRotationFL:Rotator), Scale="1,1,1") {
    rotationMode=BMM_Replace
    rotationSpace=BCS_BoneSpace
    translationSpace=BCS_BoneSpace
    alphaBoolBlend {blendOption=Linear}
    boneToModify {boneName=wheel_steering_fr_l}
   }
   n10 = "Transform (Modify) Bone - Bone: wheel_tilted_fr_r"(ComponentPose=@n9, Alpha=1, Translation="0,0,0", Rotation=Get WheelSpinFR() -> (WheelSpinFR:Rotator), Scale="1,1,1") {rotationMode=BMM_Replace, rotationSpace=BCS_BoneSpace, alphaBoolBlend {blendOption=Linear}, boneToModify {boneName=wheel_tilted_fr_r}}
   n11 = "Transform (Modify) Bone - Bone: wheel_tilted_fr_l"(ComponentPose=@n10, Alpha=1, Translation="0,0,0", Rotation=Get WheelSpinFL() -> (WheelSpinFL:Rotator), Scale="1,1,1") {rotationMode=BMM_Replace, rotationSpace=BCS_BoneSpace, alphaBoolBlend {blendOption=Linear}, boneToModify {boneName=wheel_tilted_fr_l}}
   n12 = Get WheelSpinBR() -> (WheelSpinBR:Rotator)
   n13 = "Transform (Modify) Bone - Bone: wheel_tilted_bk_r"(ComponentPose=@n11, Alpha=1, Translation="0,0,0", Rotation=@n12, Scale="1,1,1") {rotationMode=BMM_Replace, rotationSpace=BCS_BoneSpace, alphaBoolBlend {blendOption=Linear}, boneToModify {boneName=wheel_tilted_bk_r}}
   n14 = Get WheelSpinBL() -> (WheelSpinBL:Rotator)
   n15 = "Transform (Modify) Bone - Bone: wheel_tilted_bk_l"(ComponentPose=@n13, Alpha=1, Translation="0,0,0", Rotation=@n14, Scale="1,1,1") {rotationMode=BMM_Replace, rotationSpace=BCS_BoneSpace, alphaBoolBlend {blendOption=Linear}, boneToModify {boneName=wheel_tilted_bk_l}}
   n16 = "Transform (Modify) Bone - Bone: wheel_tilted_md_l"(ComponentPose=@n15, Alpha=1, Translation="0,0,0", Rotation=@n14, Scale="1,1,1") {rotationMode=BMM_Replace, rotationSpace=BCS_BoneSpace, alphaBoolBlend {blendOption=Linear}, boneToModify {boneName=wheel_tilted_md_l}}
   n17 = "Transform (Modify) Bone - Bone: wheel_tilted_md_r"(ComponentPose=@n16, Alpha=1, Translation="0,0,0", Rotation=@n12, Scale="1,1,1") {rotationMode=BMM_Replace, rotationSpace=BCS_BoneSpace, alphaBoolBlend {blendOption=Linear}, boneToModify {boneName=wheel_tilted_md_r}}
   n18 = "Transform (Modify) Bone - Bone: steering_wheel"(ComponentPose=@n17, Alpha=1, Translation="0,0,0", Rotation=Get SteeringRotation() -> (SteeringRotation:Rotator), Scale="1,1,1") {rotationMode=BMM_Replace, rotationSpace=BCS_BoneSpace, alphaBoolBlend {blendOption=Linear}, boneToModify {boneName=steering_wheel}}
   n19 = RigidBody(ComponentPose=@n18) {
    bClampLinearTranslationLimitToRefPose=true
    bOverrideWorldGravity=true
    overridePhysicsAsset=&/Semi_Truck_Veh_Dyn_dmg_mudflaps
    alphaBoolBlend {blendOption=Linear}
    overrideWorldGravity {z=-980}
    simSpaceSettings {
     maxAngularAcceleration=100
     maxAngularVelocity=100
     maxLinearAcceleration=100
     maxLinearVelocity=100
     worldAlpha=1
     externalLinearDragV=<0.3 0.3 0.3>
    }
   }
   n20 = RigidBody(ComponentPose=@n19, Alpha=Get triggerLadder() -> (triggerLadder:double)) {
    bOverrideWorldGravity=true
    evaluationResetTime=-1
    overridePhysicsAsset=&/Semi_Truck_Veh_Dyn_dmg_ladder
    alphaBoolBlend {blendOption=Linear}
    baseBoneRef {boneName=body}
    overrideWorldGravity {z=-980}
    simSpaceSettings {
     maxAngularAcceleration=500
     maxAngularVelocity=500
     maxLinearAcceleration=500
     maxLinearVelocity=500
     worldAlpha=1
     externalLinearDragV=<0.7 0.7 0.7>
    }
   }
   n21 = RigidBody(ComponentPose=@n20, Alpha=Get TriggerMirrorL() -> (TriggerMirrorL:double)) {
    bOverrideWorldGravity=true
    evaluationResetTime=-1
    overridePhysicsAsset=&/Semi_Truck_Veh_Dyn_dmg_mirror_l
    alphaBoolBlend {blendOption=Linear}
    baseBoneRef {boneName=body}
    overrideWorldGravity {z=-980}
    simSpaceSettings {
     maxAngularAcceleration=500
     maxAngularVelocity=500
     maxLinearAcceleration=500
     maxLinearVelocity=500
     worldAlpha=1
     externalLinearDragV=<0.7 0.7 0.7>
    }
   }
   n22 = RigidBody(ComponentPose=@n21, Alpha=Get TriggerMirrorR() -> (TriggerMirrorR:double)) {
    bOverrideWorldGravity=true
    evaluationResetTime=-1
    overridePhysicsAsset=&/Semi_Truck_Veh_Dyn_dmg_mirror_r
    alphaBoolBlend {blendOption=Linear}
    baseBoneRef {boneName=body}
    overrideWorldGravity {z=-980}
    simSpaceSettings {
     maxAngularAcceleration=500
     maxAngularVelocity=500
     maxLinearAcceleration=500
     maxLinearVelocity=500
     worldAlpha=1
     externalLinearDragV=<0.7 0.7 0.7>
    }
   }
   n23 = Get hasOffroadTireMod() -> (hasOffroadTireMod:bool)
   n24 = "Transform (Modify) Bone - Bone: steering_knuckle_fr_l"!(ComponentPose=@n22, bAlphaBoolEnabled:bool=@n23, Translation="0,-10,0", Rotation="0,0,0", Scale="1,1,1") -> (Pose) {
    alphaInputType=Bool
    rotationSpace=BCS_BoneSpace
    scaleMode=BMM_Additive
    scaleSpace=BCS_BoneSpace
    translationMode=BMM_Additive
    translationSpace=BCS_BoneSpace
    alphaBoolBlend {blendOption=Linear}
    boneToModify {boneName=steering_knuckle_fr_l}
    translation {y=-10}
   }
   n25 = "Transform (Modify) Bone - Bone: steering_knuckle_fr_r"!(ComponentPose=@n24, bAlphaBoolEnabled:bool=@n23, Translation="0,10,0", Rotation="0,0,0", Scale="1,1,1") -> (Pose) {
    alphaInputType=Bool
    rotationSpace=BCS_BoneSpace
    scaleMode=BMM_Additive
    scaleSpace=BCS_BoneSpace
    translationMode=BMM_Additive
    translationSpace=BCS_BoneSpace
    alphaBoolBlend {blendOption=Linear}
    boneToModify {boneName=steering_knuckle_fr_r}
    translation {y=10}
   }
   n26 = "Transform (Modify) Bone - Bone: wheel_steering_bk_l"!(ComponentPose=@n25, bAlphaBoolEnabled:bool=@n23, Translation="-15,10,0", Rotation="0,0,0", Scale="1,1,1") -> (Pose) {
    alphaInputType=Bool
    rotationSpace=BCS_BoneSpace
    scaleMode=BMM_Additive
    scaleSpace=BCS_BoneSpace
    translationMode=BMM_Additive
    translationSpace=BCS_BoneSpace
    alphaBoolBlend {blendOption=Linear}
    boneToModify {boneName=wheel_steering_bk_l}
    translation=<-15 10>
   }
   n27 = "Transform (Modify) Bone - Bone: wheel_steering_bk_r"!(ComponentPose=@n26, bAlphaBoolEnabled:bool=@n23, Translation="-15,-10,0", Rotation="0,0,0", Scale="1,1,1") -> (Pose) {
    alphaInputType=Bool
    rotationSpace=BCS_BoneSpace
    scaleMode=BMM_Additive
    scaleSpace=BCS_BoneSpace
    translationMode=BMM_Additive
    translationSpace=BCS_BoneSpace
    alphaBoolBlend {blendOption=Linear}
    boneToModify {boneName=wheel_steering_bk_r}
    translation=<-15 -10>
   }
   n28 = "Transform (Modify) Bone - Bone: wheel_steering_md_l"!(ComponentPose=@n27, bAlphaBoolEnabled:bool=@n23, Translation="-15,10,0", Rotation="0,0,0", Scale="1,1,1") -> (Pose) {
    alphaInputType=Bool
    rotationSpace=BCS_BoneSpace
    scaleMode=BMM_Additive
    scaleSpace=BCS_BoneSpace
    translationMode=BMM_Additive
    translationSpace=BCS_BoneSpace
    alphaBoolBlend {blendOption=Linear}
    boneToModify {boneName=wheel_steering_md_l}
    translation=<-15 10>
   }
   n29 = "Transform (Modify) Bone - Bone: wheel_steering_md_r"!(ComponentPose=@n28, bAlphaBoolEnabled:bool=@n23, Translation="-15,-10,0", Rotation="0,0,0", Scale="1,1,1") -> (Pose) {
    alphaInputType=Bool
    rotationSpace=BCS_BoneSpace
    scaleMode=BMM_Additive
    scaleSpace=BCS_BoneSpace
    translationMode=BMM_Additive
    translationSpace=BCS_BoneSpace
    alphaBoolBlend {blendOption=Linear}
    boneToModify {boneName=wheel_steering_md_r}
    translation=<-15 -10>
   }
   Output Pose(Result=Component To Local(ComponentPose=@n29)) {layerGroup=None}
  ]}
 EventGraph {nodes [
   Event BlueprintUpdateAnimation()
   n1 = Get Vehicle() -> (Vehicle:Fort Athena SKVehicle Object Reference)
   > Is Valid(InputObject=@n1)
   > n2 = Sequence()
   > n3 = Set Health Percent(Health Percent:double=SafeDivide(
     A=GetHealth!(self:Fort Vehicle Interface Interface=@n1) -> (ReturnValue:float)
     B=GetMaxHealth!(self:Fort Vehicle Interface Interface=@n1) -> (ReturnValue:float)
    )) -> (Output_Get:double)
   > Set triggerLadder(triggerLadder:double="To Float (Boolean)"(InBool="float <= float"!(A:double=@n3, B:double=0.25) -> (ReturnValue:bool) {is=title K2Node_CallFunction})) -> (Output_Get:double)
   > Set TriggerMirrorL(TriggerMirrorL:double="To Float (Boolean)"(InBool="float <= float"(A=@n3, B=0.65))) -> (Output_Get:double)
   > Set TriggerMirrorR(TriggerMirrorR:double="To Float (Boolean)"(InBool="float <= float"(A=@n3, B=0.7))) -> (Output_Get:double)
   n4 = Cast To FortDagwoodVehicle!(Object=@n1) -> (AsFort Dagwood Vehicle:Fort Dagwood Vehicle Object Reference, bSuccess:bool=~) {is=title K2Node_DynamicCast}
   n5 = Branch(execute=@n2.then_1, Condition=@n4.bSuccess)
   > Set hasOffroadTireMod(hasOffroadTireMod:bool=HasVehicleModApplied!(
     self:Fort Vehicle Interface Interface=@n4
     ModTag:"Gameplay Tag Structure (by ref)"=Get OffroadTireModTag() -> (OffroadTireModTag:Gameplay Tag Structure)
    ) -> (ReturnValue:bool)) -> (Output_Get:bool)
   n6 = GetMiddleAxleForwardAdjust!(self:Fort Dagwood Vehicle Object Reference=@n4) -> (ReturnValue:float)
   n7 = Get AxlePositionB()
   > Set Mid Differential Position(Mid Differential Position:Vector=vector + vector(A:Vector=MakeVector(X=@n6, Y=0), B:Vector=@n7) -> (ReturnValue:Vector) {is=title K2Node_CommutativeAssociativeBinaryOperator}) -> (Output_Get:Vector)
   Set hasOffroadTireMod(execute=@n5.else, hasOffroadTireMod=false)
   > Set Mid Differential Position(Mid Differential Position=@n7)
  ]}
]
variables [
 hasOffroadTireMod {type=boolean}
 Health Percent {type=number}
 Mid Differential Position {type=Vector}
 OffroadTireModTag {type=GameplayTag, default {tagName=Vehicle.Mod.Tire.Offroad}}
 triggerLadder {type=number}
 TriggerMirrorL {type=number}
 TriggerMirrorR {type=number}
]
