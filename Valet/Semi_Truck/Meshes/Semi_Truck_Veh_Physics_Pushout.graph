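# Physics asset for the Semi_Truck vehicle mesh, a "Pushout" variant holding a simplified collision shape.
# A single body on the frame bone carries one wide, low box named Phys_Base, slightly offset along the length of the truck.
# It defines no constraints, so the box moves rigidly with that bone.
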PhysicsAsset /Valet/Semi_Truck/Meshes/Semi_Truck_Veh_Physics_Pushout
bodies {frame [{shapeType=Box, extentX=704, extentY=241, extentZ=67, shapeName=Phys_Base, center=<-25.4432 -0.0024 5.3731>}]}
constraints []
